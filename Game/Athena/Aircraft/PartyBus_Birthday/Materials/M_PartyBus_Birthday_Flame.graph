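# Unlit additive flame material for the Birthday Party Bus aircraft.
# Vertex offset sums height-weighted sine/cosine wobble, oscillating drag and clamped-sine flicker, then adds the shared Party Bus animation function.
# Color comes from a BlackBody temperature, blending a UV tip-to-bottom gradient with an inverse-Fresnel core, then hue-shifted.
# Fresnel blends opacity with depth fade; the Mobile path drops the fade and keeps only flicker plus bus animation.
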Material /Game/Athena/Aircraft/PartyBus_Birthday/Materials/M_PartyBus_Birthday_Flame
graph {
 properties {blendMode=BLEND_Additive, pixelDepthOffsetMode=PDOM_Legacy, shadingModel=MSM_Unlit, usedShadingModels=Unlit}
 nodes [
  n1 = VectorParameter() {parameterName=Direction, defaultValue=<rgba 1 0 0.3 1>}
  n2 = Divide(A=ComponentMask(None=WorldPosition().XYZ) {b=true}, B=ScalarParameter() {defaultValue=342.8702, parameterName=Frequency})
  n3 = Time()
  n4 = Multiply(A=@n3, B=ScalarParameter() {defaultValue=1, parameterName=Speed})
  n5 = OneMinus(None=ComponentMask(None=TextureCoordinate()) {g=true})
  n6 = Multiply(A=@n5, B=ScalarParameter() {defaultValue=10.8387, parameterName=Amplitude})
  n7 = Multiply(A=Sine(None=Add(A=@n2, B=@n4)), B=@n6)
  n8 = VectorParameter() {parameterName=Direction_2, defaultValue=<rgba 0 1 -0.3 1>}
  n9 = Multiply(A=@n3, B=ScalarParameter() {defaultValue=0.7524, parameterName=Speed_2})
  n10 = Multiply(A=@n6, B=Cosine(None=Add(A=@n2, B=@n9)))
  n11 = ScalarParameter() {defaultValue=3, parameterName=DragFalloff}
  n12 = Time()
  n13 = ScalarParameter() {defaultValue=1, parameterName=Speed_Drag}
  n14 = Add(A=Sine(None=Multiply(A=@n12, B=@n13)))
  n15 = Multiply(
   A=Transform(None=VectorParameter() {parameterName=Direction_3, defaultValue=<rgba 0 -1 0 1>}.RGB) {transformSourceType=TRANSFORMSOURCE_Local}
   B=ScalarParameter() {defaultValue=41.6634, parameterName=Drag}
  )
  n16 = Add(A=Add(A=Multiply(A=@n1.RGB, B=@n7), B=Multiply(A=@n8.RGB, B=@n10)), B=Multiply(A=Power(Base=@n5, Exp=@n11), B=Multiply(A=@n14, B=@n15)))
  n17 = Constant3Vector() {constant=<rgba 0 0 1 0>}
  n18 = Multiply(A=Time(), B=ScalarParameter() {defaultValue=2.6839, parameterName=Speed_Flicker})
  n19 = Clamp(None=Sine(None=@n18))
  n20 = ComponentMask(None=TextureCoordinate()) {g=true}
  n21 = Multiply(A=Multiply(A=Multiply(A=@n17, B=@n19), B=OneMinus(None=@n20)), B=ScalarParameter() {defaultValue=-18.8313, parameterName=Flicker})
  Add(
   A=ShadingPathSwitch(Default=Add(A=@n16, B=@n21), Mobile=@n21) {desc=Migrated from FeatureLevelSwitch}
   B=MaterialFunctionCall() {materialFunction=/Game/Athena/Aircraft/PartyBus_RT/Materials/MF_PartyBusAnimationWPO}.Result
  )
  Frac(None=@n18)
  n22 = LinearInterpolate(
   A=ScalarParameter() {defaultValue=1844.9662, parameterName=Temp_Tip}
   B=ScalarParameter() {defaultValue=55.141, parameterName=Temp_bottom}
   Alpha=ComponentMask(None=TextureCoordinate()) {g=true}
  )
  n23 = Fresnel(ExponentIn=ScalarParameter() {defaultValue=0.5407, parameterName=Fresnel}) {baseReflectFraction=0}
  n24 = Multiply(A=ScalarParameter() {defaultValue=2218.1697, parameterName=Temp_Core}, B=OneMinus(None=@n23))
  MaterialFunctionCall(Hue Shift Percentage=ScalarParameter() {parameterName=Hue}, Texture=BlackBody!(Temp=Add(A=@n22, B=@n24))) {materialFunction=&HueShift}
  n25 = LinearInterpolate(A=ScalarParameter() {defaultValue=1, parameterName=Opacity_1}, B=ScalarParameter() {parameterName=Opacity_2}, Alpha=@n23)
  ShadingPathSwitch(Default=DepthFade(Opacity=@n25, FadeDistance=ScalarParameter() {defaultValue=79.3591, parameterName=DepthFade}), Mobile=@n25) {desc=Migrated from FeatureLevelSwitch}
 ]
}
